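; Manager that shows or hides the Figment S06 Butterfly level in response to the stateful global Butterfly ShowLevel event.
; A scalable-float toggle chooses between data layer switching and level injection.
; The data layer path sets the runtime state to Activated or Loaded, so a hidden level stays loaded.
; The injection path removes any previous instance by GUID, then injects at the origin when visible and stores the new GUID.
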(event EventBeginPlay
  (bind (_asynctaskproxy _payload) (GameplayEventRouter|StartListeningToStatefulEvent (GameplayEventRouter|GetGlobalEventRouter) "/Figment_S06/Gameplay/EventRouter/Figment_Butterfly_ShowLevel_Event.Figment_Butterfly_ShowLevel_Event")
    (:OnEventReceived
      _)
    (:OnSavedEventStateExists
      _))
  (if (ScalableFloat|GetValueasBool (Variables|Default|GetUseDataLayersToInject))
    (bind _payload_showlevel_1_2ab9bcbd4ba571cf04a5299c4f74bbd9 (Utilities|Struct|BreakFigmentButterflyShowLevelEvent _payload))
    (CallFunction|SetVisibilitywithDataLayer _payload_showlevel_1_2ab9bcbd4ba571cf04a5299c4f74bbd9)
    (else
      (CallFunction|SetVisibilitywithInjection _payload_showlevel_1_2ab9bcbd4ba571cf04a5299c4f74bbd9))))

(fn SetVisibilityWithDataLayer
  (params (Index :type Boolean))
  (bind _returnvalue (WorldPartition|GetDataLayerManager))
  (DataLayers|SetDataLayerRuntimeState _returnvalue (Variables|Default|GetButteflyLevelDataLayer) (select Index "Activated" "Loaded") false :!class /Script/Engine.DataLayerManager))

(fn SetVisibilityWithInjection
  (params (Visible :type Boolean))
  (CallFunction|RemoveInjectedLevel (Variables|Default|GetInjected%20Level%20GUID))
  (if Visible
    (CallFunction|RemoveInjectedLevel (Variables|Default|GetInjected%20Level%20GUID))
    (bind _injectedlevelguid (CallFunction|RequesttoInjectLevel (Variables|Default|GetButterflyLevelInjectionData) (Math|Transform|MakeTransform "0, 0, 0")))
    (Variables|Default|SetInjected%20Level%20GUID _injectedlevelguid)))
